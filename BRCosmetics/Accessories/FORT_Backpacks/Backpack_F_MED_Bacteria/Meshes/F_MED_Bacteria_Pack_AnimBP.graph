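# Animation blueprint for the Bacteria backpack cosmetic, adding secondary motion to its two knife pockets and three flaps.
# Anim Dynamics simulates each bone, then additive Transform Bone nodes layer on flap and pocket roll and slightly raise the pack.
# The update event derives per-flap open angles from vertical velocity, clamped by skydiving state, with fixer rolls cancelling them at zero velocity.
# Both knife pockets receive a corrective roll during skydive floating.
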AnimBlueprint /BRCosmetics/Accessories/FORT_Backpacks/Backpack_F_MED_Bacteria/Meshes/F_MED_Bacteria_Pack_AnimBP
target_skeleton=&/F_MED_Bacteria_Pack_Skeleton
graphs [
 AnimGraph {nodes [
   n1 = Anim Dynamics - Bone: knife_pocket_r(ComponentPose=Local To Component(LocalPose=Local Space Ref Pose()), Alpha=1, ExternalForce="0,0,0") {
    angularDampingOverride=0.94
    angularSpringConstant=100
    bAngularSpring=true
    bGravityOverrideInSimSpace=true
    bOverrideAngularDamping=true
    alphaBoolBlend {blendOption=Linear}
    boundBone {boneName=knife_pocket_r}
    gravityOverride {y=-100}
    physicsBodyDefinitions [{
      sphereCollisionRadius=0
      boundBone {boneName=knife_pocket_r}
      boxExtents=<95 30 125>
      constraintSetup {angularLimitsMax=<0 0 10>, angularLimitsMin=<0 0 -10>, angularTarget=<0 0 0>}
      localJointOffset=<10 0 0>
     }]
    relativeSpaceBone {boneName=root}
   }
   n2 = Anim Dynamics - Bone: knife_pocket_l(ComponentPose=@n1, Alpha=1, ExternalForce="0,0,0") {
    angularDampingOverride=0.99
    angularSpringConstant=1000
    bOverrideAngularDamping=true
    alphaBoolBlend {blendOption=Linear}
    boundBone {boneName=knife_pocket_l}
    gravityOverride {y=-980}
    physicsBodyDefinitions [{
      sphereCollisionRadius=0
      boundBone {boneName=knife_pocket_l}
      boxExtents=<40 100 40>
      constraintSetup {angularLimitsMax=<0 0 10>, angularLimitsMin=<0 0 -10>, angularTarget=<0 0 0>}
      localJointOffset=<20 0 0>
     }]
    relativeSpaceBone {boneName=root}
   }
   n3 = Anim Dynamics - Bone: flap_1(ComponentPose=@n2, Alpha=1, ExternalForce="0,0,0") {
    angularDampingOverride=0.99
    angularSpringConstant=100
    bAngularSpring=true
    bOverrideAngularDamping=true
    alphaBoolBlend {blendOption=Linear}
    boundBone {boneName=flap_1}
    physicsBodyDefinitions [{
      sphereCollisionRadius=0
      boundBone {boneName=flap_1}
      boxExtents=<50 50 50>
      constraintSetup {angularLimitsMax=<0 0 0>, angularLimitsMin=<0 0 -10>, angularTarget=<0 0 0>}
      localJointOffset=<3.2 10 0>
     }]
   }
   n4 = Anim Dynamics - Bone: flap_2(ComponentPose=@n3, Alpha=1, ExternalForce="0,0,0") {
    angularDampingOverride=0.99
    angularSpringConstant=100
    bAngularSpring=true
    bOverrideAngularDamping=true
    alphaBoolBlend {blendOption=Linear}
    boundBone {boneName=flap_2}
    physicsBodyDefinitions [{
      sphereCollisionRadius=0
      boundBone {boneName=flap_2}
      boxExtents=<50 50 50>
      constraintSetup {angularLimitsMax=<0 0 0>, angularLimitsMin=<0 0 -10>, angularTarget=<0 0 0>}
      localJointOffset=<10 2.6 0>
     }]
   }
   n5 = Anim Dynamics - Bone: flap_3(ComponentPose=@n4, Alpha=1, ExternalForce="0,0,0") {
    angularDampingOverride=0.99
    angularSpringConstant=100
    bAngularSpring=true
    bOverrideAngularDamping=true
    alphaBoolBlend {blendOption=Linear}
    boundBone {boneName=flap_3}
    physicsBodyDefinitions [{
      sphereCollisionRadius=0
      boundBone {boneName=flap_3}
      boxExtents=<10 10 10>
      constraintSetup {angularLimitsMax=<0 0 0>, angularLimitsMin=<0 0 -10>, angularTarget=<0 0 0>}
      localJointOffset=<5 1.2 0>
     }]
   }
   n6 = Get flap_1_fixer() -> (flap_1_fixer:double)
   n7 = "Transform (Modify) Bone - Bone: flap_2"!(
    ComponentPose="Transform (Modify) Bone - Bone: flap_1"!(
     ComponentPose="Transform (Modify) Bone - Bone: flap_1"(ComponentPose=@n5, Alpha=1, Translation="0,0,0", Rotation_Roll=@n6, Rotation_Pitch=0, Rotation_Yaw=0, Scale="1,1,1") {rotationMode=BMM_Additive, alphaBoolBlend {blendOption=Linear}, boneToModify {boneName=flap_1}}
     Alpha=1
     Translation="0,0,0"
     Rotation_Roll:float=Get Open_angle_1() -> (Open_angle_1:double)
     Rotation_Pitch:float=0
     Rotation_Yaw:float=0
     Scale="1,1,1"
    ) -> (Pose) {rotationMode=BMM_Additive, alphaBoolBlend {blendOption=Linear}, boneToModify {boneName=flap_1}}
    Alpha=1
    Translation="0,0,0"
    Rotation_Roll:float=Get flap_2_fixer() -> (flap_2_fixer:double)
    Rotation_Pitch:float=0
    Rotation_Yaw:float=0
    Scale="1,1,1"
   ) -> (Pose) {rotationMode=BMM_Additive, alphaBoolBlend {blendOption=Linear}, boneToModify {boneName=flap_2}}
   n8 = Get Open_angle_2() -> (Open_angle_2:double)
   n9 = "Transform (Modify) Bone - Bone: flap_3"!(
    ComponentPose="Transform (Modify) Bone - Bone: flap_3"(
     ComponentPose="Transform (Modify) Bone - Bone: flap_2"(ComponentPose=@n7, Alpha=1, Translation="0,0,0", Rotation_Roll=@n8, Rotation_Pitch=0, Rotation_Yaw=0, Scale="1,1,1") {rotationMode=BMM_Additive, alphaBoolBlend {blendOption=Linear}, boneToModify {boneName=flap_2}}
     Alpha=1
     Translation="0,0,0"
     Rotation_Roll=Get flap_3_fixer() -> (flap_3_fixer:double)
     Rotation_Pitch=0
     Rotation_Yaw=0
     Scale="1,1,1"
    ) {rotationMode=BMM_Additive, alphaBoolBlend {blendOption=Linear}, boneToModify {boneName=flap_3}}
    Alpha=1
    Translation="0,0,0"
    Rotation_Roll:float=Get Open_angle_3() -> (Open_angle_3:double)
    Rotation_Pitch:float=0
    Rotation_Yaw:float=0
    Scale="1,1,1"
   ) -> (Pose) {rotationMode=BMM_Additive, alphaBoolBlend {blendOption=Linear}, boneToModify {boneName=flap_3}}
   n10 = Get Knife_Skydyve_floating_correct() -> (Knife_Skydyve_floating_correct:double)
   n11 = "Transform (Modify) Bone - Bone: pack"(
    ComponentPose="Transform (Modify) Bone - Bone: knife_pocket_l"!(
     ComponentPose="Transform (Modify) Bone - Bone: knife_pocket_r"!(ComponentPose=@n9, Alpha=1, Translation="0,0,0", Rotation_Roll:float=@n10, Rotation_Pitch:float=0, Rotation_Yaw:float=0, Scale="1,1,1") -> (Pose) {rotationMode=BMM_Additive, alphaBoolBlend {blendOption=Linear}, boneToModify {boneName=knife_pocket_r}}
     Alpha=1
     Translation="0,0,0"
     Rotation_Roll:float=@n10
     Rotation_Pitch:float=0
     Rotation_Yaw:float=0
     Scale="1,1,1"
    ) -> (Pose) {rotationMode=BMM_Additive, alphaBoolBlend {blendOption=Linear}, boneToModify {boneName=knife_pocket_l}}
    Translation="0,0,2"
    Rotation="0,0,0"
   ) {translationMode=BMM_Additive, alphaBoolBlend {blendOption=Linear}, boneToModify {boneName=pack}, translation {z=2}}
   Output Pose(Result=Component To Local(ComponentPose=@n11)) {layerGroup=None}
  ]}
 EventGraph {nodes [
   Event BlueprintUpdateAnimation()
   n1 = Get PawnVelocityZ()
   n2 = float + float(B=-0.015)
   n3 = float + float(B=-600)
   n4 = SelectFloat(A=5, B=10, bPickA=Get bIsSkyDiving())
   n5 = "Clamp (Float)"(Value=float * float(A=float + float(B=@n3), B=@n2) {is=title K2Node_PromotableOperator}, Min=0, Max=@n4)
   > Set flap_1_fixer(flap_1_fixer:double=SelectFloat(A=0, B=float * float(A=@n5, B=-1), bPickA="Not Equal (Float)"(A=@n1))) -> (Output_Get:double)
   n6 = float + float(A=@n3, B=@n1)
   > Set Open_angle_1(Open_angle_1:double="Clamp (Float)"(Value=float * float(A=@n6, B=@n2), Min=0, Max=@n4)) -> (Output_Get:double)
   n7 = float + float(B=-0.01)
   n8 = float * float(A=float * float(A=float + float(B=@n3), B=@n7), B=-1)
   n9 = float * float(A=@n4, B=-1)
   n10 = Get PawnVelocityZ()
   > Set flap_2_fixer(flap_2_fixer:double=SelectFloat(A=0, B="Clamp (Float)"(Value=@n8, Min=@n9, Max=0), bPickA="Not Equal (Float)"(A=@n10))) -> (Output_Get:double)
   n11 = float + float(A=@n3, B=@n10)
   > Set Open_angle_2(Open_angle_2:double="Clamp (Float)"(Value=float * float(A=@n11, B=@n7), Min=0, Max=@n4)) -> (Output_Get:double)
   n12 = float + float(B=-0.04)
   n13 = "Clamp (Float)"(Value=float * float(A=float + float(B=@n3), B=@n12), Min=0, Max=@n4)
   n14 = Get PawnVelocityZ()
   > Set flap_3_fixer(flap_3_fixer:double=SelectFloat(A=0, B=float * float(A=@n13, B=-1), bPickA="Not Equal (Float)"(A=@n14))) -> (Output_Get:double)
   n15 = float + float(A=@n14, B=@n3)
   > Set Open_angle_3(Open_angle_3:double="Clamp (Float)"(Value=float * float(A=@n15, B=@n12), Min=0, Max=@n4)) -> (Output_Get:double)
   > Set Knife_Skydyve_floating_correct(Knife_Skydyve_floating_correct:double=SelectFloat(A=40, B=0, bPickA=Get bIsSkydiveFloating())) -> (Output_Get:double)
  ]}
]
variables [
 flap_1_fixer {type=number}
 flap_2_fixer {type=number}
 flap_3_fixer {type=number}
 Knife_Skydyve_floating_correct {type=number}
 Open_angle_1 {type=number}
 Open_angle_2 {type=number}
 Open_angle_3 {type=number}
 Skydiving_crouching_alpha {type=number}
]
